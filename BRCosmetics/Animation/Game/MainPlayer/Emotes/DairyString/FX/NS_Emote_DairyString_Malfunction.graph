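# Niagara malfunction effect for the DairyString player emote, pairing blown mesh particles tinted by a smoke color with sparks.
# BlowingParticles spawns in short repeated loops along a half-cylinder, pushed out along shape normals with curl noise, drag, growing scale and fading alpha.
# Classic_Sparks bursts when User.BurstIndex changes, gated by system age; the index picks count, offset and direction for falling, flickering sprites.
# A third emitter, ToonSmoke001, has no renderers.
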NiagaraSystem /BRCosmetics/Animation/Game/MainPlayer/Emotes/DairyString/FX/NS_Emote_DairyString_Malfunction
dependencies {
 usedDataInterfaces [&NiagaraDataInterfaceCurve, &NiagaraDataInterfaceColorCurve, &NiagaraDataInterfaceVectorCurve]
 usedDynamicInputs [&FNTimeOfDayColor]
 usedModules [
  &self:SystemSpawnScript.SystemScriptSource.SystemScriptGraph.NiagaraNodeAssignment_0.SetVariables_#1
  &SystemState
 ]
}
emitters {
 BlowingParticles {
  emitterUpdateScript [
   &EmitterState(
    Life Cycle Mode=::NewEnumerator1
    Loop Behavior=::NewEnumerator2
    Loop Count=9
    Loop Delay=0.15
    Loop Duration=0.24
    Recalculate Duration Each Loop=true
    UseLoopDelay=true
   )
   &SpawnBurst_Instantaneous(Use Spawn Probability=false, UseLoopCountLimit=false) {enabled=false}
   &SpawnRate(SpawnRate=42.2507, Use Spawn Probability=false)
   /SetVariables(Emitter.RandomInt:NiagaraInt32=&RandomRangeInt(Minimum=0, Maximum=999)) {isSetParameters=true}
  ]
  particleSpawnScript [
   &PartitionParticles(Number Of Partitions=1) {enabled=false}
   &InitializeParticle(
    Color=$System.SmokeColor
    Lifetime=&RandomRangeFloat(Minimum=0.3, Maximum=0.7)
    Mass Max=1.5
    Mass Mode=::NewEnumerator3
    Mesh Scale Mode=::NewEnumerator1
    Mesh Uniform Scale Max=0.8
    Mesh Uniform Scale Min=0.5
    Position Offset=<0 -4.8799 -17.6863>
    Sprite Rotation Mode=::NewEnumerator2
    UsePositionOffset=true
    Write Material Random=true
   )
   &ShapeLocation(
    Cylinder Height=38.8357
    Cylinder Radius=20
    Fixed Random Seed=true
    Hemicircle Internal Cap=false
    Hemicircle X=true
    MaskSpawnGroup=false
    Override Seed=true
    Random Seed:NiagaraInt32=$Emitter.RandomInt
    Randomness Mode:ENiagaraRandomnessMode=::NewEnumerator1
    Rotation Mode=::NewEnumerator2
    Shape Primitive=::NewEnumerator1
    Surface Only=true
    Use Endcaps In Surface Only Mode=false
    UseOwnerScale=true
    Yaw / Pitch / Roll=<0.25 0.25 0>
   ) {module=ShapeLocation001}
   &InitialMeshOrientation(Use Rotation Vector=false)
   &AddVelocity(
    Velocity=&AddVectorToPosition(Position=Linked() {linkedVariable {name=Particles.ShapeLocation001.ShapeNormal, type=Vector3f}}, Vector=<0 0 0.3>)
    Velocity Speed Scale=70
   )
   &ShapeLocation(MaskSpawnGroup=false, Override Seed=false, Sphere Radius=3, UseOwnerScale=true)
  ]
  particleUpdateScript [
   &ParticleState()
   &UpdateMeshOrientation(Rotation Rate=0.1, Rotation Vector=<1 1 1>)
   &ScaleMeshSize(Scale Factor=&VectorFromCurve(VectorCurve=DataInterface() {propertyValues {
       XCurve {
        defaultValue=1
        keys [
         {arriveTangent=3.6936, interpMode=RCIM_Cubic, leaveTangent=3.6936, tangentMode=RCTM_User, time=0, value=0}
         {arriveTangent=0.3083, interpMode=RCIM_Cubic, leaveTangent=0.3083, tangentMode=RCTM_User, time=0.5146, value=1.1342}
         <at 1 1.3 RCIM_Cubic>
        ]
       }
       YCurve {keys [<at 0 1>,<at 1 1>]}
       ZCurve {keys [<at 0 1>,<at 1 1>]}
      }})) {enabled=false}
   &ScaleMeshSize(Scale Factor=&VectorFromFloat(Value=&FloatFromCurve(FloatCurve=DataInterface() {propertyValues {Curve {
         defaultValue=1
         keys [
          {arriveTangent=0.4035, interpMode=RCIM_Cubic, leaveTangent=0.4035, tangentMode=RCTM_User, time=0, value=0.2}
          {arriveTangent=1.1279, interpMode=RCIM_Cubic, leaveTangent=1.1279, time=0.7468, value=0.8905}
          {
           arriveTangent=0.6848
           arriveTangentWeight=0.2109
           interpMode=RCIM_Cubic
           leaveTangent=0.6848
           tangentMode=RCTM_User
           time=0.7744
           value=1.0735
          }
          {arriveTangent=0.1288, arriveTangentWeight=0.1228, interpMode=RCIM_Cubic, leaveTangent=0.1288, tangentMode=RCTM_User, time=1.0012, value=1.2}
         ]
        }}}))) {module=ScaleMeshSize001}
   &CurlNoiseForce(
    Bias Noise=false
    Noise Frequency=84.0353
    Noise Strength=&FloatFromCurve(
     FloatCurve=DataInterface() {propertyValues {Curve {
        defaultValue=1
        keys [
         {arriveTangent=4.1785, interpMode=RCIM_Cubic, leaveTangent=4.1785, tangentMode=RCTM_User, time=0, value=0}
         {arriveTangent=1.1179, interpMode=RCIM_Cubic, leaveTangent=1.1179, tangentMode=RCTM_User, time=0.2042, value=0.6228}
         {arriveTangent=0.0224, interpMode=RCIM_Cubic, leaveTangent=0.0224, tangentMode=RCTM_User, time=1, value=1}
        ]
       }}}
     Scale Curve=164.6343
    )
    Pan Noise=false
   )
   &Drag(
    Drag=&FloatFromCurve(
     FloatCurve=DataInterface() {propertyValues {Curve {
        defaultValue=1
        keys [
         {arriveTangent=4.1785, interpMode=RCIM_Cubic, leaveTangent=4.1785, tangentMode=RCTM_User, time=0, value=0}
         {arriveTangent=1.1179, interpMode=RCIM_Cubic, leaveTangent=1.1179, tangentMode=RCTM_User, time=0.2042, value=0.6228}
         {arriveTangent=0.0224, interpMode=RCIM_Cubic, leaveTangent=0.0224, tangentMode=RCTM_User, time=1, value=1}
        ]
       }}}
     Scale Curve=2
    )
    Use Linear Drag=true
    Use Rotational Drag=true
   )
   &SolveForcesAndVelocity(Clamp Velocity=false, Limit Acceleration=false)
   &ScaleColor(
    Linear Color Curve=DataInterface() {propertyValues {
      AlphaCurve {keys [<at 0.0088 1>,<at 0.2211 0.6278>,<at 1 0>]}
      BlueCurve {keys [<at 0 1>,<at 1 1>]}
      GreenCurve {keys [<at 0 1>,<at 1 1>]}
      RedCurve {keys [<at 0 1>,<at 1 1>]}
     }}
    Scale Mode=2
   )
   &DynamicMaterialParameters(
    Index 0 Param 1=&FloatFromCurve(FloatCurve=DataInterface() {propertyValues {Curve {
        defaultValue=0
        keys [
         {arriveTangent=4.1785, interpMode=RCIM_Cubic, leaveTangent=4.1785, tangentMode=RCTM_User, time=0, value=0}
         {arriveTangent=1.1179, interpMode=RCIM_Cubic, leaveTangent=1.1179, tangentMode=RCTM_User, time=0.2042, value=0.6228}
         {arriveTangent=0.0224, interpMode=RCIM_Cubic, leaveTangent=0.0224, tangentMode=RCTM_User, time=1, value=1}
        ]
       }}})
    Index 0 Param 2=0.328
    Module.Index0Float0Write=true
    Module.Index0Float1Write=true
    Module.Index0Float2Write=true
    Module.Index0Float3Write=true
   )
  ]
  rendererClasses [&NiagaraMeshRendererProperties]
  renderers [{rendererClass=&NiagaraMeshRendererProperties, rendererIndex=0}]
 }
 Classic_Sparks {
  emitterUpdateScript [
   &EmitterState(Life Cycle Mode=::NewEnumerator1, Loop Behavior=::NewEnumerator1, Loop Delay=2.3833, UseLoopDelay=false)
   /Game/Effects/Niagara/Modules/Logic/UpdateValueStateChange(InputIntParam:NiagaraInt32=&MakeIntFromFloat(Float=$User.BurstIndex, Conversion Mode=::NewEnumerator1))
   &self:TriggerSpawn(
    Spawn Count:NiagaraInt32=&MergeHelper_IntSwitch(
     0=8
     1=5
     2=6
     3=5
     4=7
     5=?
     Index=&MakeIntFromFloat(Float=$User.BurstIndex, Conversion Mode=::NewEnumerator1)
     Num=::NewEnumerator3
    )
    Spawn Group:NiagaraInt32=&MakeIntFromFloat(Float=0, Conversion Mode=::NewEnumerator1)
    Trigger:NiagaraBool=&Boolean_AndOperation(Bool1=$Emitter.UpdateValueStateChange.Changed, Bool2=&SetBoolByFloatComparison(A=$System.Age, B=0.56))
   ) {module=TriggerSpawn}
   &SpawnRate(SpawnRate=25.8466, Use Spawn Probability=true) {enabled=false}
  ]
  particleSpawnScript [
   &InitializeParticle(
    Color Mode=::NewEnumerator0
    Lifetime=&RandomRangeFloat(Minimum=0.15, Maximum=0.3)
    Mesh Scale Max=<0.0663 0.0663 0.0663>
    Mesh Scale Min=<2.4885 2.8084 2.6334>
    Mesh Scale Mode=::NewEnumerator2
    Position Mode=::NewEnumerator1
    Position Offset=&MergeHelper_VectorSwitch(
     0=<7.0954 10.5663 0>
     1=<-10.0536 8 -6.3944>
     2=<0 7.7711 -3.8>
     3=<3.0404 -15.7676 -12.1414>
     4=<0 0 0>
     Index=&MakeIntFromFloat(Float=$User.BurstIndex)
     Num=::NewEnumerator3
    )
    Position Offset Coordinate Space=::World
    Sprite Rotation Mode=::NewEnumerator2
    Sprite Size Max=<3 6>
    Sprite Size Min=<1 1>
    Sprite Size Mode=::NewEnumerator2
    UsePositionOffset=true
    Write Material Random=true
   )
   &AddVelocity(
    Cone Angle=41.5318
    Cone Axis=<1 1 1>
    Module.UseConeFalloff=false
    Module.UseVelDistribution=false
    Override Seed=false
    Rotation Coordinate Space=::World
    Rotation Mode=::NewEnumerator2
    Velocity Mode=::NewEnumerator2
    Velocity Speed=&MaskFloatBySpawnGroup(Masked Float=250, Passthrough Float=0)
    Yaw / Pitch / Roll:Vector3f=&MergeHelper_VectorSwitch(
     0=<0 0 -40.361>
     1=<335.3159 -605.4348 1341.7222>
     2=<335.3159 -605.4348 1341.7222>
     3=<0 0 -40.361>
     4=<0 0 -40.361>
     Index=&MakeIntFromFloat(Float=$User.BurstIndex)
     Num=::NewEnumerator3
    )
   ) {module=AddVelocity001}
  ]
  particleUpdateScript [
   &ParticleState()
   /SetVariables(
    Particles.Flicker:NiagaraFloat=&Waveform("Waveform [1]"=::NewEnumerator0, "[1] Frequency"=75, "[1] Amplitude Min/Max"=<0 2>)
    Particles.Flicker Size:NiagaraFloat=&Waveform("Waveform [1]"=::NewEnumerator0, "[1] Frequency"=75, "[1] Amplitude Min/Max"=<0.5 1>)
    Particles.Flicker Value:NiagaraFloat=&Waveform("Waveform [1]"=::NewEnumerator0, "[1] Frequency"=75, "[1] Amplitude Min/Max"=<2 8>)
   ) {isSetParameters=true}
   &ScaleSpriteSize(
    Scale Sprite Size Mode=::NewEnumerator0
    Uniform Scale Factor=&Lerp_Float(
     A=1
     B=&Multiply_Float(A=$Particles.Flicker Size, B=1)
     Alpha=&FloatFromCurve(FloatCurve=DataInterface() {propertyValues {Curve {keys [<at 0 0.1076>,<at 0.2485 0.6128>,<at 0.45 1>,<at 1 1>]}}})
    )
   )
   &ScaleColor(
    Scale Mode=::NewEnumerator1
    Scale RGBA=&Multiply_Vector4(
     A=&MakeVector4_FromLinearColor(LinearColor=&ColorFromCurve(ColorCurve=DataInterface() {propertyValues {
         AlphaCurve {keys [<at 0.122 1>]}
         BlueCurve {keys [<at 0 1.0417>,<at 0.1355 0.9911>,<at 0.3032 0>,<at 1 0>]}
         GreenCurve {keys [<at 0 7.4334>,<at 0.1355 10.4768>,<at 0.3032 9.1355>,<at 1 3.6718>]}
         RedCurve {keys [<at 0 20>,<at 0.1355 16>,<at 0.3032 13>,<at 1 9>]}
        }}))
     B=&MakeVector4_FromLinearColor(LinearColor=&MakeLinearColorFromVectorAndFloat(
       "Vector (RGB)"=&Add_Vector(
        A=<1 1 1>
        B=/Niagara/DynamicInputs/TypeConversions/MakeVectorFromLinearColorHSV(Color:LinearColor=/Niagara/DynamicInputs/LinearColor/MakeLinearColorFromHSV(
          HSV From:NiagaraInt32=1
          HSV:Vector3f=?
          Hue:NiagaraFloat=0
          Saturation:NiagaraFloat=0
          Value:NiagaraFloat=&Lerp_Float(
           A=0
           B=$Particles.Flicker Value
           Alpha=&FloatFromCurve(FloatCurve=DataInterface() {propertyValues {Curve {keys [<at 0 0>,<at 0.5 0>,<at 0.55 0.9624>,<at 1 1>]}}})
          )
          Alpha:NiagaraFloat=1
         ))
       )
       "Float (Alpha)"=&Lerp_Float(
        A=1
        B=$Particles.Flicker
        Alpha=&FloatFromCurve(FloatCurve=DataInterface() {propertyValues {Curve {keys [<at 0 0>,<at 0.25 0>,<at 0.35 1>,<at 1 1>]}}})
       )
      ))
    )
    ScaleRGBA=true
   )
   &GravityForce(Gravity=&RandomRangeVector(Minimum=<0 0 -286.0087>, Maximum=<0 0 -887.0334>))
   &SolveForcesAndVelocity(Clamp Velocity=false, Limit Acceleration=false)
   &ScaleSpriteSizeBySpeed(
    Max Scale Factor=&RandomRangeVector2D(Minimum=<0.22 18.0452>, Maximum=<6 11.7483>)
    Min Scale Factor=<0.15 1>
    Velocity Threshold=3800
   )
   /SetVariables() {isSetParameters=true}
  ]
  rendererClasses [&NiagaraSpriteRendererProperties]
  renderers [{rendererClass=&NiagaraSpriteRendererProperties}]
 }
 ToonSmoke001 {rendererClasses [], renderers []}
}
summary {
 systemName=NS_Emote_DairyString_Malfunction
 emitters [
  {emitterName=Classic_Sparks, rendererClasses [&NiagaraSpriteRendererProperties]}
  {emitterName=BlowingParticles, rendererClasses [&NiagaraMeshRendererProperties]}
  {emitterName=ToonSmoke001, rendererClasses []}
 ]
 userVariables [
  User.BurstIndex {type=NiagaraFloat, defaultValue=1, description=""}
  User.RibbonIDOffset {type=NiagaraInt32}
 ]
}
