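# Niagara system for the DairyString emote's green ray laser, with a charge-up, beam and impact.
# GatheringFlare builds a pulsing mesh, glow and light with swirling motes, then BlowingParticles draws wobbling Hermite-spline ribbons from the owner to a target socket.
# Delayed bursts at that socket add ImpactFlare flashes and light, plus golden sparks from BlowingParticles001.
# User.RayIndex selects per-ray offsets, timing, sizes and counts; quality level scales some particle counts.
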NiagaraSystem /BRCosmetics/Animation/Game/MainPlayer/Emotes/DairyString/FX/NS_Emote_DairyString_RayLaser
dependencies {
 usedDataInterfaces [
  &NiagaraDataInterfaceCurve
  &NiagaraDataInterfaceArrayColor
  &NiagaraDataInterfaceVectorCurve
  &NiagaraDataInterfaceSpriteRendererInfo
  &NiagaraDataInterfaceColorCurve
  &NiagaraDataInterfaceSkeletalMesh
  &NiagaraDataInterfaceVector2DCurve
  &NiagaraDataInterfaceSocketReader
 ]
 usedDynamicInputs [
  &Add_Float
  &Subtract_Float
  &FindQuatBetweenTwoVectors
  &Subtract_Vector
 ]
 usedModules [
  &self:SystemSpawnScript.SystemScriptSource.SystemScriptGraph.NiagaraNodeAssignment_4.SetVariables_#1
  &self:SystemSpawnScript.SystemScriptSource.SystemScriptGraph.NiagaraNodeAssignment_5.SetVariables_#2
  &self:SystemSpawnScript.SystemScriptSource.SystemScriptGraph.NiagaraNodeAssignment_0.SetVariables_#3
  &self:SystemSpawnScript.SystemScriptSource.SystemScriptGraph.NiagaraNodeAssignment_6.SetVariables_#4
  &SystemState
  &self:SystemSpawnScript.SystemScriptSource.SystemScriptGraph.NiagaraNodeAssignment_1.SetVariables_#5
  &self:SystemSpawnScript.SystemScriptSource.SystemScriptGraph.NiagaraNodeAssignment_2.SetVariables_#6
  &SampleSkeletalMesh
  &self:SystemSpawnScript.SystemScriptSource.SystemScriptGraph.NiagaraNodeAssignment_3.SetVariables_#7
 ]
}
emitters {
 BlowingParticles {
  emitterUpdateScript [
   &EmitterState(Life Cycle Mode=::NewEnumerator1, Loop Behavior=::NewEnumerator1, Loop Delay=$System.LoopDuration, Loop Duration=0.5, UseLoopDelay=true)
   &SpawnRate(SpawnRate=62.4378, Use Spawn Probability=false)
  ]
  particleSpawnScript [&InitializeParticle(
    Color=<rgba 0 1.5 0.6992 1>
    Lifetime=0.1
    Mass Max=1.5
    Mass Mode=::NewEnumerator3
    Ribbon Width Mode=::NewEnumerator1
    Sprite Rotation Mode=::NewEnumerator2
    Sprite Size Mode=::NewEnumerator1
    Uniform Sprite Size Max=8
    Uniform Sprite Size Min=4
    UsePositionOffset=false
    Write Material Random=true
   )]
  particleUpdateScript [
   &SocketLocation(
    Position Sampling=::NewEnumerator1
    SamplePosOffset=true
    Sampled Position Offset=&MergeHelper_VectorSwitch(
     0=<0 0 5>
     1=<0 0 24>
     2=<0 0 15>
     3=?
     Index=&MakeIntFromFloat(Float=$User.RayIndex)
     Num=::NewEnumerator1
    )
    Socket Index=1
    Socket Reader=$System.Socket Reader
    Socket Sampling Mode=::NewEnumerator2
    VelocitySpeedLimit=false
   ) {module=SocketLocation001}
   /Game/Effects/Niagara/Modules/Curves/FN_HermiteSpline(
    Position Blend:NiagaraFloat=$Particles.NormalizedAge
    Position End:NiagaraPosition=$Output.SocketLocation001.SampledPosition
    Position Start:NiagaraPosition=$Engine.Owner.Position
    Write Particles Position=true
   )
   &ParticleState()
   /SetVariables(Particles.ParticleLength:NiagaraFloat=&Multiply_Float(
     A=&Abs_Float(Float:NiagaraFloat=&VectorLength(Vector=Linked() {linkedVariable {name=Particles.Position, type=NiagaraPosition}}))
     B=0.1
    )) {isSetParameters=true}
   /SetVariables(Particles.Position=&AddVectorToPosition(
     Position=Linked() {linkedVariable {name=Output.FN_HermiteSpline.Position, type=Vector3f}}
     Vector=&Multiply_VectorByFloat(
      Vector=&MakeVector(
       X=0
       Y=&Waveform(
        Global Input Position=&Multiply_Float(A=$Particles.ParticleLength, B=0.005)
        Global Amplitude Scale=10
        "Waveform [1]"=::NewEnumerator0
        "[1] Frequency"=1
        "[1] Amplitude Min/Max"=<0 1>
        "[1] Scale"=0.13
        "[2] Phase"=$Emitter.Age
        "[3] Phase"=$Emitter.Age
       )
       Z=&Waveform(
        Global Input Position=&Multiply_Float(A=$Particles.ParticleLength, B=0.453)
        Global Amplitude Scale=10
        "Waveform [1]"=::NewEnumerator0
        "[1] Frequency"=1.5584
        "[1] Amplitude Min/Max"=<-1 1>
        "[1] Scale"=0.1347
        "[2] Phase"=$Emitter.Age
        "[3] Phase"=$Emitter.Age
       )
      )
      Float=&FloatFromCurve(FloatCurve=DataInterface() {propertyValues {Curve {keys [
           {interpMode=RCIM_Cubic, tangentMode=RCTM_User, time=0, value=0}
           {arriveTangent=1.6386, interpMode=RCIM_Cubic, leaveTangent=1.6386, time=0.2035, value=0.6382}
           {arriveTangent=-0.8013, interpMode=RCIM_Cubic, leaveTangent=-0.8013, time=0.5546, value=0.9088}
           <at 1 0>
          ]}}})
     )
    )) {isSetParameters=true}
   &SolveForcesAndVelocity(Clamp Velocity=false, Limit Acceleration=false)
   &DynamicMaterialParameters(
    Module.Index0Float0Write=true
    Module.Index0Float1Write=true
    Module.Index0Float2Write=true
    Module.Index0Float3Write=true
    Module.Index1Float0Write=true
    Module.Index1Float1Write=true
    Module.Index1Float2Write=true
    Module.Index1Float3Write=true
    Write Parameter Index 1=::NewEnumerator1
   )
   /SetVariables(
    Particles.GlowWidth:NiagaraFloat=&MergeHelper_FloatSwitch(
     0=&FloatFromCurve(FloatCurve=DataInterface() {propertyValues {Curve {keys [<at 1.4636 1>,<at 1.5405 0>]}}}, CurveIndex=$Emitter.Age, Scale Curve=5)
     1=&FloatFromCurve(FloatCurve=DataInterface() {propertyValues {Curve {keys [<at 1.0936 1>,<at 1.1704 0>]}}}, CurveIndex=$Emitter.Age, Scale Curve=5)
     2=&FloatFromCurve(FloatCurve=DataInterface() {propertyValues {Curve {keys [<at 2.1327 1>,<at 2.2096 0>]}}}, CurveIndex=$Emitter.Age, Scale Curve=5)
     3=?
     Index=&MakeIntFromFloat(Float=$User.RayIndex)
     Num=::NewEnumerator1
    )
    Particles.RibbonWidthLightning:NiagaraFloat=&MergeHelper_FloatSwitch(
     0=&FloatFromCurve(FloatCurve=DataInterface() {propertyValues {Curve {keys [<at 1.4636 1>,<at 1.5405 0>]}}}, CurveIndex=$Emitter.Age, Scale Curve=3)
     1=&FloatFromCurve(FloatCurve=DataInterface() {propertyValues {Curve {keys [<at 1.0936 1>,<at 1.1704 0>]}}}, CurveIndex=$Emitter.Age, Scale Curve=3)
     2=&FloatFromCurve(FloatCurve=DataInterface() {propertyValues {Curve {keys [<at 2.1327 1>,<at 2.2096 0>]}}}, CurveIndex=$Emitter.Age, Scale Curve=3)
     3=?
     Index=&MakeIntFromFloat(Float=$User.RayIndex)
     Num=::NewEnumerator1
    )
   ) {isSetParameters=true}
  ]
  rendererClasses [&NiagaraRibbonRendererProperties]
  renderers [
   {rendererClass=&NiagaraRibbonRendererProperties}
   {rendererClass=&NiagaraRibbonRendererProperties, rendererIndex=1}
  ]
 }
 BlowingParticles001 {
  emitterUpdateScript [
   &EmitterState(Life Cycle Mode=::NewEnumerator1, Loop Behavior=::NewEnumerator1, Loop Delay=$System.PopDelay, Loop Duration=1, UseLoopDelay=true)
   &SpawnBurst_Instantaneous(
    Spawn Count=&MergeHelper_IntSwitch(0=10, 1=10, 2=0, 3=?, 4=?, 5=?, Index=&MakeIntFromFloat(Float=$User.RayIndex), Num=::NewEnumerator1)
    Use Spawn Probability=false
    UseLoopCountLimit=false
   )
  ]
  particleSpawnScript [
   &InitializeParticle(
    Color=<rgba 0.7396 0.5119 0.1233 1>
    Lifetime=&RandomRangeFloat(Minimum=0.1, Maximum=0.45)
    Mass Max=1.5
    Mass Mode=::NewEnumerator3
    Sprite Rotation Mode=::NewEnumerator2
    Sprite Size Mode=::NewEnumerator1
    Uniform Sprite Size Max=3
    Uniform Sprite Size Min=1
    UsePositionOffset=false
    Write Material Random=true
   )
   &SocketLocation(
    SamplePosOffset=true
    Sampled Position Offset=&MergeHelper_VectorSwitch(
     0=<0 0 5>
     1=<0 0 5>
     2=<0 -5 20>
     3=?
     Index=&MakeIntFromFloat(Float=$User.RayIndex)
     Num=::NewEnumerator1
    )
    Sampled Position Offset Coordinate Space:ENiagaraSkelMeshTransforms=::NewEnumerator2
    Socket Index=1
    Socket Reader=$System.Socket Reader
    Socket Sampling Mode=::NewEnumerator2
    VelocitySpeedLimit=false
   ) {module=SocketLocation001}
   &ShapeLocation(
    MaskSpawnGroup=false
    Override Seed=false
    Ring Radius=&MergeHelper_FloatSwitch(0=15, 1=6, 2=6, 3=0, 4=0, Index=&MakeIntFromFloat(Float=$User.RayIndex), Num=::NewEnumerator3)
    Rotation Mode=::NewEnumerator2
    Shape Primitive=::NewEnumerator4
    UseOwnerScale=true
    Yaw / Pitch / Roll=<0 0 -0.096>
   ) {module=ShapeLocation001}
   &AddVelocity(
    Module.Use Falloff Distance=false
    Origin Offset=<0 8 -14.3373>
    Override Seed=false
    Velocity Mode=::NewEnumerator1
    Velocity Origin=$Output.SocketLocation001.SampledPosition
    Velocity Speed=213.8895
   )
  ]
  particleUpdateScript [
   &Drag(
    Drag=&FloatFromCurve(
     FloatCurve=DataInterface() {propertyValues {Curve {
        defaultValue=1
        keys [
         {arriveTangent=4.1785, interpMode=RCIM_Cubic, leaveTangent=4.1785, tangentMode=RCTM_User, time=0, value=0}
         {arriveTangent=1.1179, interpMode=RCIM_Cubic, leaveTangent=1.1179, tangentMode=RCTM_User, time=0.2042, value=0.6228}
         {arriveTangent=0.0224, interpMode=RCIM_Cubic, leaveTangent=0.0224, tangentMode=RCTM_User, time=1, value=1}
        ]
       }}}
     Scale Curve=&RandomRangeFloat(Minimum=20, Maximum=25)
    )
    Use Linear Drag=true
    Use Rotational Drag=true
   )
   &InheritSourceMovement(ApplyForce=false, ApplyPosition=true, ApplyPositionFalloff=false)
   &SolveForcesAndVelocity(Clamp Velocity=false, Limit Acceleration=false)
   &ParticleState()
   &ScaleSpriteSize(Uniform Curve Sprite Scale=DataInterface() {propertyValues {Curve {defaultValue=0, keys [{tangentMode=RCTM_User, time=0, value=1}, {tangentMode=RCTM_User, time=1, value=0}]}}}) {module=ScaleSpriteSize001}
   &ScaleSpriteSizeBySpeed(Max Scale Factor=<1 7.5935>, Min Scale Factor=<1 10.4312>)
   &ScaleSpriteSize(Uniform Curve Sprite Scale=DataInterface() {propertyValues {Curve {keys [
        {arriveTangent=14.5835, interpMode=RCIM_Cubic, leaveTangent=14.5835, tangentMode=RCTM_User, time=0, value=0}
        {arriveTangent=1.1111, interpMode=RCIM_Cubic, leaveTangent=1.1111, time=0.1, value=1}
        {arriveTangent=-1.1111, interpMode=RCIM_Cubic, leaveTangent=-1.1111, time=0.9, value=1}
        {arriveTangent=-8.7113, interpMode=RCIM_Cubic, leaveTangent=-8.7113, tangentMode=RCTM_User, time=1, value=0}
       ]}}})
  ]
  rendererClasses [&NiagaraSpriteRendererProperties]
  renderers [{rendererClass=&NiagaraSpriteRendererProperties}]
 }
 Empty {rendererClasses [], renderers []}
 GatheringFlare {
  emitterUpdateScript [
   &EmitterState(
    Inactive Response=::NewEnumerator1
    Life Cycle Mode=::NewEnumerator1
    Loop Behavior=::NewEnumerator1
    Loop Duration=$System.LoopDuration
    UseLoopDelay=false
   )
   &SpawnBurst_Instantaneous(Spawn Time=0.12, Use Spawn Probability=false, UseLoopCountLimit=false)
   &SpawnRate(
    SpawnGroup=1
    SpawnRate=&FloatFromCurve(FloatCurve=DataInterface() {propertyValues {Curve {keys [<at 0.17 0>,<at 0.1748 0.9206>]}}}, CurveIndex=$Emitter.LoopedAge, Scale Curve=6)
    Use Spawn Probability=false
   )
   &SpawnRate(
    SpawnGroup=2
    SpawnRate=&MergeHelper_FloatSwitch(
     0=&MakeFloatFromInt(Int=&MergeHelper_IntSwitch(0=80, 1=100, 2=150, 3=150, 4=150, 5=?, Index=$Engine.QualityLevel, Num=::NewEnumerator3))
     1=&MakeFloatFromInt(Int=&MergeHelper_IntSwitch(0=80, 1=100, 2=150, 3=150, 4=150, 5=?, Index=$Engine.QualityLevel, Num=::NewEnumerator3))
     2=&FloatFromCurve(
      FloatCurve=DataInterface() {propertyValues {Curve {defaultValue=1, keys [{tangentMode=RCTM_User, time=0, value=0}, <at 0.8968 2.0915>, {tangentMode=RCTM_User, time=0.9958, value=0.301}]}}}
      CurveIndex=$Emitter.NormalizedLoopAge
      Scale Curve=&MakeFloatFromInt(Int=&MergeHelper_IntSwitch(0=70, 1=80, 2=120, 3=120, 4=120, 5=?, Index=$Engine.QualityLevel, Num=::NewEnumerator3))
     )
     3=?
     Index=&MakeIntFromFloat(Float=$User.RayIndex)
     Num=::NewEnumerator1
    )
    Use Spawn Probability=false
   ) {module=SpawnRate001}
   &SpawnBurst_Instantaneous(Spawn Group=3, Spawn Time=0.12, Use Spawn Probability=false, UseLoopCountLimit=false) {module=SpawnBurst_Instantaneous001}
  ]
  particleSpawnScript [
   /SetVariables(Particles.VisibilityTag=&SpawnGroup()) {isSetParameters=true, module=SetVariables_#8}
   &InitializeParticle(
    Color=&MergeHelper_ColorSwitch(
     0=<rgba 0 1.5 0.6992 1>
     1=&SelectLinearColorFromArray(
      Array Sampling Mode=::NewEnumerator0
      Direct Array Index=?
      Color Selection Array=DataInterface() {propertyValues {ColorData [<rgba 0.0539 1 0 1>, <rgba 0 1.7676 2 1>]}}
      Random Seed=0
     )
     2=<rgba 0 3 1.5002 1>
     3=?
     Index=&SpawnGroup()
     Num=::NewEnumerator1
    )
    Lifetime=&MergeHelper_FloatSwitch(0=1.8, 1=0.3, 2=0.438, 3=1.8, Index=&SpawnGroup(), Num=::NewEnumerator2)
    Mesh Scale Mode=::NewEnumerator3
    Mesh Uniform Scale=0.216
    Position Offset=<0 2 0>
    Sprite Rotation Mode=::NewEnumerator2
    Sprite Size Mode=::NewEnumerator3
    Uniform Sprite Size=&MaskFloatBySpawnGroup(
     Masked Float=&RandomRangeFloat(Minimum=0.3, Maximum=0.8)
     Passthrough Float=35.1543
     Spawn Group Mask=2
    )
    UsePositionOffset=true
    Write Material Random=true
   )
   &CameraOffset(Camera Offset Amount=20)
   &ShapeLocation(MaskSpawnGroup=true, Override Seed=false, Spawn Group Mask=2, Sphere Radius=35.1789, UseOwnerScale=true)
   &AddVelocity(
    Module.Use Falloff Distance=false
    Origin Offset=<0 2 0>
    Override Seed=false
    Velocity Mode=::NewEnumerator1
    Velocity Speed=&MaskFloatBySpawnGroup(Masked Float=-150, Passthrough Float=0, Spawn Group Mask=2)
   )
  ]
  particleUpdateScript [
   &ParticleState()
   &ScaleMeshSize(Scale Factor=&VectorFromCurve(
     VectorCurve=DataInterface() {propertyValues {
       XCurve {
        defaultValue=1
        keys [
         {tangentMode=RCTM_User, time=0, value=0}
         {interpMode=RCIM_Cubic, leaveTangent=55.6408, tangentMode=RCTM_Break, time=0.9, value=1}
         {arriveTangent=-10, interpMode=RCIM_Cubic, leaveTangent=-10, time=0.9344, value=0.6603}
         {interpMode=RCIM_Cubic, tangentMode=RCTM_Break, time=1, value=0}
        ]
       }
       YCurve {
        defaultValue=1
        keys [
         {tangentMode=RCTM_User, time=0, value=0}
         {interpMode=RCIM_Cubic, leaveTangent=55.6408, tangentMode=RCTM_Break, time=0.9, value=1}
         {arriveTangent=-10, interpMode=RCIM_Cubic, leaveTangent=-10, time=0.9348, value=0.6305}
         {interpMode=RCIM_Cubic, tangentMode=RCTM_Break, time=1, value=0}
        ]
       }
       ZCurve {
        defaultValue=1
        keys [
         {tangentMode=RCTM_User, time=0, value=0}
         {interpMode=RCIM_Cubic, leaveTangent=55.6408, tangentMode=RCTM_Break, time=0.9, value=1}
         {arriveTangent=-10, interpMode=RCIM_Cubic, leaveTangent=-10, time=0.9348, value=0.6305}
         {interpMode=RCIM_Cubic, tangentMode=RCTM_Break, time=1, value=0}
        ]
       }
      }}
     CurveIndex=$Emitter.NormalizedLoopAge
     Scale Curve=&VectorFromFloat(Value=&FloatFromCurve(
       FloatCurve=DataInterface() {propertyValues {Curve {keys [<at 0 1>,<at 0.9 1>,<at 1 0>]}}}
       CurveIndex=$Emitter.NormalizedLoopAge
       Scale Curve=&Waveform("Waveform [1]"=::NewEnumerator10, "[1] Frequency"=7.7946, "[1] Amplitude Min/Max"=<0.7085 1>)
      ))
    ))
   &VortexForce(
    UseOriginOffset=true
    Vortex Force Amount=&MaskFloatBySpawnGroup(
     Masked Float=&FloatFromCurve(FloatCurve=DataInterface() {propertyValues {Curve {keys [<at 0 1>,<at 0.7026 0.85>,<at 1.0064 0>]}}}, Scale Curve=839.0593)
     Passthrough Float=0
     Spawn Group Mask=2
    )
   )
   &SolveForcesAndVelocity(Clamp Velocity=false, Limit Acceleration=false)
   &ScaleSpriteSizeBySpeed(
    Max Scale Factor=&MaskVector2dBySpawnGroup(Masked Vector2D=<1 4>, Passthrough Vector2D=<1 1>, Spawn Group Mask=2)
    Min Scale Factor=&MaskVector2dBySpawnGroup(Masked Vector2D=<1 5>, Passthrough Vector2D=<1 1>, Spawn Group Mask=2)
   )
   &SubUVAnimation(
    "Sprite Renderer "=DataInterface() {propertyValues {SpriteRenderer=&self:GatheringFlare_0.Renderer}}
    UseEndFrame=false
    UseStartFrame=false
   )
   /SetVariables() {isSetParameters=true, module=SetVariables_#9}
   /SetVariables(Particles.GlowSize=&Vector2DFromFloat(Value=&Multiply_Float(
      A=&FloatFromCurve(
       FloatCurve=DataInterface() {propertyValues {Curve {
          defaultValue=1
          keys [
           {tangentMode=RCTM_User, time=0, value=0.59}
           {interpMode=RCIM_Cubic, leaveTangent=29.5816, tangentMode=RCTM_Break, time=0.9, value=1.8217}
           {tangentMode=RCTM_User, time=1, value=0}
          ]
         }}}
       CurveIndex=$Emitter.NormalizedLoopAge
       Scale Curve=30
      )
      B=&FloatFromCurve(
       FloatCurve=DataInterface() {propertyValues {Curve {keys [<at 0 1>,<at 0.9 1>,<at 1 0>]}}}
       CurveIndex=$Emitter.NormalizedLoopAge
       Scale Curve=&Waveform("Waveform [1]"=::NewEnumerator10, "[1] Frequency"=7.7946, "[1] Amplitude Min/Max"=<0.7085 1>)
      )
     ))) {isSetParameters=true, module=SetVariables_#10}
   &FN_Light_Attributes(
    Color=&ColorFromCurve(
     ColorCurve=DataInterface() {propertyValues {
       AlphaCurve {keys [<at 0 1 RCIM_Cubic>,<at 1 0 RCIM_Cubic>]}
       BlueCurve {keys [<at 0 0.6992 RCIM_Cubic>,<at 0.9874 0>]}
       GreenCurve {keys [<at 0 1.5 RCIM_Cubic>,<at 0.9874 0>]}
       RedCurve {keys [<at 0 0 RCIM_Cubic>,<at 0.9874 0>]}
      }}
     CurveIndex=$Emitter.NormalizedLoopAge
    )
    Create Color Attribute=true
    Create Radius Attribute=true
    Radius=35
   )
   /SetVariables(Particles.Mesh2Size:Vector3f=&VectorFromCurve(
     VectorCurve=DataInterface() {propertyValues {
       XCurve {
        defaultValue=1
        keys [
         {tangentMode=RCTM_User, time=0, value=0}
         {interpMode=RCIM_Cubic, leaveTangent=89.199, tangentMode=RCTM_Break, time=0.9081, value=0.7013}
         {arriveTangent=-7.6297, interpMode=RCIM_Cubic, leaveTangent=-7.6297, time=0.9425, value=0.3616}
         {interpMode=RCIM_Cubic, tangentMode=RCTM_Break, time=1, value=0}
        ]
       }
       YCurve {
        defaultValue=1
        keys [
         {tangentMode=RCTM_User, time=0, value=0}
         {interpMode=RCIM_Cubic, leaveTangent=89.199, tangentMode=RCTM_Break, time=0.9081, value=0.7013}
         {arriveTangent=-7.6297, interpMode=RCIM_Cubic, leaveTangent=-7.6297, time=0.9429, value=0.3318}
         {interpMode=RCIM_Cubic, tangentMode=RCTM_Break, time=1, value=0}
        ]
       }
       ZCurve {
        defaultValue=1
        keys [
         {tangentMode=RCTM_User, time=0, value=0}
         {interpMode=RCIM_Cubic, leaveTangent=89.199, tangentMode=RCTM_Break, time=0.9081, value=0.7013}
         {arriveTangent=-7.6297, interpMode=RCIM_Cubic, leaveTangent=-7.6297, time=0.9429, value=0.3318}
         {interpMode=RCIM_Cubic, tangentMode=RCTM_Break, time=1, value=0}
        ]
       }
      }}
     CurveIndex=$Emitter.NormalizedLoopAge
     Scale Curve=&VectorFromFloat(Value=&FloatFromCurve(
       FloatCurve=DataInterface() {propertyValues {Curve {keys [<at 0 1>,<at 0.9 1>,<at 1 0>]}}}
       CurveIndex=$Emitter.NormalizedLoopAge
       Scale Curve=&Waveform("Waveform [1]"=::NewEnumerator10, "[1] Frequency"=7.7946, "[1] Amplitude Min/Max"=<0.7085 1>)
      ))
    )) {isSetParameters=true}
  ]
  rendererClasses [
   &NiagaraSpriteRendererProperties
   &NiagaraLightRendererProperties
   &NiagaraMeshRendererProperties
  ]
  renderers [
   {rendererClass=&NiagaraSpriteRendererProperties}
   {rendererClass=&NiagaraSpriteRendererProperties, rendererIndex=1}
   {rendererClass=&NiagaraSpriteRendererProperties, rendererIndex=2}
   {rendererClass=&NiagaraLightRendererProperties, rendererIndex=3}
   {rendererClass=&NiagaraMeshRendererProperties, rendererIndex=4}
   {rendererClass=&NiagaraMeshRendererProperties, rendererIndex=5}
  ]
 }
 ImpactFlare {
  emitterUpdateScript [
   &EmitterState(Life Cycle Mode=::NewEnumerator1, Loop Behavior=::NewEnumerator1, Loop Delay=$System.Delay, Loop Duration=1, UseLoopDelay=true)
   &SpawnBurst_Instantaneous(Spawn Group=1, Use Spawn Probability=false, UseLoopCountLimit=false) {module=SpawnBurst_Instantaneous001}
   &SpawnBurst_Instantaneous(Use Spawn Probability=false, UseLoopCountLimit=false)
   &SpawnBurst_Instantaneous(Spawn Count=12, Spawn Group=2, Use Spawn Probability=false, UseLoopCountLimit=false) {module=SpawnBurst_Instantaneous002}
   &SpawnBurst_Instantaneous(
    Spawn Count=&MergeHelper_IntSwitch(
     0=&MergeHelper_IntSwitch(0=13, 1=17, 2=30, 3=30, 4=30, 5=?, Index=$Engine.QualityLevel, Num=::NewEnumerator3)
     1=&MergeHelper_IntSwitch(0=13, 1=17, 2=30, 3=30, 4=30, 5=?, Index=$Engine.QualityLevel, Num=::NewEnumerator3)
     2=0
     3=?
     4=?
     5=?
     Index=&MakeIntFromFloat(Float=$User.RayIndex)
     Num=::NewEnumerator1
    )
    Spawn Group=3
    Use Spawn Probability=false
    UseLoopCountLimit=false
   ) {module=SpawnBurst_Instantaneous003}
  ]
  particleSpawnScript [
   /SetVariables(Particles.VisibilityTag=&SpawnGroup()) {isSetParameters=true, module=SetVariables_#8}
   &InitializeParticle(
    Color=&MergeHelper_ColorSwitch(
     0=<rgba 0 1.5 0.6992 1>
     1=<rgba 0 3 1.5002 1>
     2=&SelectLinearColorFromArray(
      Array Sampling Mode=::NewEnumerator0
      Direct Array Index=?
      Color Selection Array=DataInterface() {propertyValues {ColorData [<rgba 0 3 1.5002 1>, <rgba 0.0368 1 0.0595 1>]}}
      Random Seed=0
     )
     3=&SelectLinearColorFromArray(
      Array Sampling Mode=::NewEnumerator0
      Direct Array Index=?
      Color Selection Array=DataInterface() {propertyValues {ColorData [<rgba 0 3 1.5002 1>, <rgba 0.0368 1 0.0595 1>]}}
      Random Seed=0
     )
     Index=&SpawnGroup()
     Num=::NewEnumerator2
    )
    Lifetime=&MergeHelper_FloatSwitch(
     0=0.5
     1=&MaskFloat(
      Mask Source=1
      Mask Parameter=&MakeIntFromFloat(Float=$User.RayIndex)
      Mask=2
      Masked Float=0.5
      Passthrough Float=0.2
     )
     2=&RandomRangeFloat(Minimum=0.1, Maximum=0.2)
     3=&RandomRangeFloat(Minimum=0.8, Maximum=0.6)
     4=0.88
     Index=&SpawnGroup()
     Num=::NewEnumerator3
    )
    Position Offset=<0 0 10>
    Position Offset Coordinate Space=::Local
    Sprite Rotation Mode=::NewEnumerator2
    Sprite Size Mode=::NewEnumerator3
    Uniform Sprite Size=&MergeHelper_FloatSwitch(0=41.6583, 1=358.4063, 2=?, 3=?, Index=&SpawnGroup(), Num=::NewEnumerator0)
    UsePositionOffset=true
    Write Material Random=true
   )
   &CameraOffset(Camera Offset Amount=30)
   &ShapeLocation(MaskSpawnGroup=true, Override Seed=false, Spawn Group Mask=3, Sphere Radius=9, UseOwnerScale=true)
   &SocketLocation(
    SamplePosOffset=true
    Sampled Position Offset=&MergeHelper_VectorSwitch(
     0=<0 0 5>
     1=<0 0 24>
     2=<0 -5 20>
     3=?
     Index=&MakeIntFromFloat(Float=$User.RayIndex)
     Num=::NewEnumerator1
    )
    Sampled Position Offset Coordinate Space:ENiagaraSkelMeshTransforms=::NewEnumerator2
    Socket Index=1
    Socket Reader=$System.Socket Reader
    Socket Sampling Mode=::NewEnumerator2
    VelocitySpeedLimit=false
   ) {module=SocketLocation001}
   &AddVelocity(
    Module.Use Falloff Distance=false
    Override Seed=false
    Velocity Mode=::NewEnumerator1
    Velocity Speed=&MaskFloatBySpawnGroup(Masked Float=6.5157, Passthrough Float=0, Spawn Group Mask=2)
   )
   &AddVelocity(
    Module.Use Falloff Distance=false
    Override Seed=false
    Velocity Mode=::NewEnumerator1
    Velocity Origin=$Output.SocketLocation001.SampledPosition
    Velocity Speed=&MergeHelper_FloatSwitch(0=0, 1=0, 2=0, 3=450, 4=450, Index=&SpawnGroup(), Num=::NewEnumerator3)
   ) {module=AddVelocity001}
   /SetVariables(Particles.StreakAlign:Vector3f=$Particles.ShapeLocation.ShapeNormal) {isSetParameters=true}
  ]
  particleUpdateScript [
   &SkeletalMeshLocation(
    Bone Sampling Mode=::NewEnumerator4
    MaskSpawnGroup=false
    Skeletal Mesh=DataInterface() {propertyValues {
      DefaultMesh=null
      ExcludeBoneName=root
      OptionalComponentClass=null
      PreviewMesh=/BRCosmetics/Animation/Props/Gadgets_DairyString/Meshes/DairyString_Gadget
      SkinningMode=SkinOnTheFly
      SoftSourceActor=null
      SourceMode=Default
      UvSetIndex=0
      WholeMeshLOD=-1
      bExcludeBone=false
      bReadDeformedGeometry=true
      bRequireCurrentFrameData=true
      ComponentTags []
      FilteredBones [JNT_Banana_1]
      FilteredSockets []
      MeshUserParameter {parameter {name=None, typeDefHandle {registeredTypeIndex=1}, varData []}}
      SamplingRegions []
     }}
   ) {enabled=false}
   &ShapeLocation(MaskSpawnGroup=true, Override Seed=false, Spawn Group Mask=4, Sphere Radius=6.91, Sphere Surface Distribution=1, UseOwnerScale=true) {module=ShapeLocation001}
   &Drag(
    Drag=&FloatFromCurve(
     FloatCurve=DataInterface() {propertyValues {Curve {
        defaultValue=1
        keys [
         {arriveTangent=4.1785, interpMode=RCIM_Cubic, leaveTangent=4.1785, tangentMode=RCTM_User, time=0, value=0}
         {arriveTangent=1.1179, interpMode=RCIM_Cubic, leaveTangent=1.1179, tangentMode=RCTM_User, time=0.2042, value=0.6228}
         {arriveTangent=0.0224, interpMode=RCIM_Cubic, leaveTangent=0.0224, tangentMode=RCTM_User, time=1, value=1}
        ]
       }}}
     Scale Curve=&RandomRangeFloat(Minimum=14.0132, Maximum=19.8906)
    )
    Use Linear Drag=true
    Use Rotational Drag=true
   )
   &InheritSourceMovement(ApplyForce=false, ApplyPosition=true, ApplyPositionFalloff=false)
   &CurlNoiseForce(
    Bias Noise=false
    Noise Frequency=27.7223
    Noise Strength=&MaskFloatBySpawnGroup(
     Masked Float=&FloatFromCurve(
      FloatCurve=DataInterface() {propertyValues {Curve {
         defaultValue=1
         keys [
          {arriveTangent=4.1785, interpMode=RCIM_Cubic, leaveTangent=4.1785, tangentMode=RCTM_User, time=0, value=0}
          {arriveTangent=1.1179, interpMode=RCIM_Cubic, leaveTangent=1.1179, tangentMode=RCTM_User, time=0.2042, value=0.6228}
          {arriveTangent=0.0224, interpMode=RCIM_Cubic, leaveTangent=0.0224, tangentMode=RCTM_User, time=1, value=1}
         ]
        }}}
      Scale Curve=500
     )
     Passthrough Float=0
     Spawn Group Mask=3
    )
    Pan Noise=false
   )
   &SolveForcesAndVelocity(Clamp Velocity=false, Limit Acceleration=false)
   &ParticleState()
   &ScaleSpriteSize(
    Scale Sprite Size Mode=::NewEnumerator0
    Uniform Scale Factor=&MergeHelper_FloatSwitch(
     0=&MakeFloatFromVector2d(Vector 2D=&Vector2DFromFloat(Value=&FloatFromCurve(FloatCurve=DataInterface() {propertyValues {Curve {
           defaultValue=1
           keys [
            {interpMode=RCIM_Cubic, leaveTangent=4.354, tangentMode=RCTM_Break, time=0, value=0.2}
            {arriveTangent=0.8588, interpMode=RCIM_Cubic, leaveTangent=0.8588, tangentMode=RCTM_User, time=0.2851, value=1.6226}
            {arriveTangent=-0.4169, tangentMode=RCTM_User, time=1, value=2}
           ]
          }}})))
     1=&MakeFloatFromVector2d(Vector 2D=&Vector2DFromFloat(Value=&FloatFromCurve(FloatCurve=DataInterface() {propertyValues {Curve {
           defaultValue=1
           keys [
            {interpMode=RCIM_Cubic, leaveTangent=4.354, tangentMode=RCTM_Break, time=0, value=0.2}
            {arriveTangent=0.8588, interpMode=RCIM_Cubic, leaveTangent=0.8588, tangentMode=RCTM_User, time=0.2851, value=1.6226}
            {arriveTangent=-0.4169, tangentMode=RCTM_User, time=1, value=2}
           ]
          }}})))
     2=&FloatFromCurve(FloatCurve=DataInterface() {propertyValues {Curve {keys [<at 0 1>,<at 1 0>]}}})
     3=?
     Index=&SpawnGroup()
     Num=::NewEnumerator1
    )
   )
   /SetVariables() {isSetParameters=true, module=SetVariables_#9}
   /SetVariables(Particles.GlowSize=&Vector2DFromFloat(Value=&FloatFromCurve(
      FloatCurve=DataInterface() {propertyValues {Curve {
         defaultValue=1
         keys [
          {interpMode=RCIM_Cubic, leaveTangent=4.354, tangentMode=RCTM_Break, time=0, value=0.2}
          {arriveTangent=-6.293, interpMode=RCIM_Cubic, leaveTangent=-6.293, tangentMode=RCTM_User, time=0.2851, value=1.6226}
          {arriveTangent=-0.4169, tangentMode=RCTM_User, time=0.7722, value=0}
         ]
        }}}
      Scale Curve=&MergeHelper_FloatSwitch(0=60, 1=90, 2=90, 3=?, Index=&MakeIntFromFloat(Float=$User.RayIndex), Num=::NewEnumerator1)
     ))) {isSetParameters=true, module=SetVariables_#10}
   /SetVariables(Particles.FlareSize:Vector2f=/Niagara/DynamicInputs/Helpers/MaskVector2d(
     Mask Source:NiagaraInt32=1
     Mask Parameter:NiagaraInt32=&MakeIntFromFloat(Float=$User.RayIndex)
     Mask:NiagaraInt32=2
     Masked Vector2D:Vector2f=&Vector2DFromFloat(Value=&FloatFromCurve(
       FloatCurve=DataInterface() {propertyValues {Curve {
          defaultValue=1
          keys [
           {interpMode=RCIM_Cubic, leaveTangent=3.8251, tangentMode=RCTM_Break, time=0, value=0.2}
           {arriveTangent=1.1979, interpMode=RCIM_Cubic, leaveTangent=1.1979, tangentMode=RCTM_User, time=0.2391, value=2.0725}
           {arriveTangent=-0.0483, interpMode=RCIM_Cubic, leaveTangent=-0.0483, tangentMode=RCTM_User, time=0.5684, value=2.1488}
           {arriveTangent=-13.9709, interpMode=RCIM_Cubic, leaveTangent=-13.9709, tangentMode=RCTM_User, time=0.6204, value=1.7462}
           {arriveTangent=-0.4169, tangentMode=RCTM_User, time=0.9333, value=0}
          ]
         }}}
       Scale Curve=120
      ))
     Passthrough Vector2D:Vector2f=&Vector2DFromFloat(Value=&FloatFromCurve(
       FloatCurve=DataInterface() {propertyValues {Curve {
          defaultValue=1
          keys [
           {interpMode=RCIM_Cubic, leaveTangent=4.354, tangentMode=RCTM_Break, time=0, value=0.2}
           {arriveTangent=0.8588, interpMode=RCIM_Cubic, leaveTangent=0.8588, tangentMode=RCTM_User, time=0.2216, value=2.0725}
           {arriveTangent=-0.4169, tangentMode=RCTM_User, time=0.9924, value=0.2003}
          ]
         }}}
       Scale Curve=&MergeHelper_FloatSwitch(0=90, 1=120, 2=?, 3=?, Index=&MakeIntFromFloat(Float=$User.RayIndex), Num=::NewEnumerator0)
      ))
    )) {isSetParameters=true}
   /SetVariables(Particles.StreakSize:Vector2f=&Vector2DFromCurve(
     Vector2Curve=DataInterface() {propertyValues {XCurve {keys [<at 0.5 1>,<at 1 0>]}, YCurve {keys [<at 0 0.2>,<at 0.5 1>]}}}
     ScaleCurve=<3 50>
    )) {isSetParameters=true}
   /SetVariables(
    Particles.DynParam1:Vector4f=<1 0 0 0>
    Particles.ParticleSize:Vector2f=&Multiply_Vector2D_ByFloat(
     Float=&FloatFromCurve(FloatCurve=DataInterface() {propertyValues {Curve {keys [<at 0.4473 1>,<at 1 0>]}}})
     Vector2D=&MakeVector2D(
      X=1
      Y=&Multiply_Float(A=&VectorLength(Vector=$Particles.Velocity), B=0.1)
     )
    )
   ) {isSetParameters=true}
   &DynamicMaterialParameters(
    Index 0 Param 1=6.0238
    Index 0 Param 2=1
    Index 0 Param 3=0.3113
    Index 0 Param 4=1
    Module.Index0Float0Write=true
    Module.Index0Float1Write=true
    Module.Index0Float2Write=true
    Module.Index0Float3Write=true
    Module.Index1Float0Write=true
    Module.Index1Float1Write=true
    Module.Index1Float2Write=true
    Module.Index1Float3Write=true
    Write Parameter Index 1=::NewEnumerator1
   )
   &FN_Light_Attributes(
    Color=&ColorFromCurve(
     ColorCurve=DataInterface() {propertyValues {
       AlphaCurve {keys [<at 0 1 RCIM_Cubic>,<at 1 0 RCIM_Cubic>]}
       BlueCurve {keys [<at 0 0.6992 RCIM_Cubic>,<at 0.9874 0>]}
       GreenCurve {keys [<at 0 1.5 RCIM_Cubic>,<at 0.9874 0>]}
       RedCurve {keys [<at 0 0 RCIM_Cubic>,<at 0.9874 0>]}
      }}
     CurveIndex=$Emitter.NormalizedLoopAge
    )
    Create Color Attribute=true
    Create Radius Attribute=true
    Radius=35
   )
   &SubUVAnimation(
    Evaluation Type=::NewEnumerator0
    "Sprite Renderer "=DataInterface() {propertyValues {SpriteRenderer=&self:ImpactFLare_0.Renderer}}
    SubUV Animation Mode:ENiagara_SubUVLookupModeV2=::NewEnumerator2
    UseEndFrame=false
    UseStartFrame=false
   )
  ]
  rendererClasses [&NiagaraSpriteRendererProperties, &NiagaraLightRendererProperties]
  renderers [
   {rendererClass=&NiagaraSpriteRendererProperties}
   {rendererClass=&NiagaraSpriteRendererProperties, rendererIndex=1}
   {rendererClass=&NiagaraLightRendererProperties, rendererIndex=2}
   {rendererClass=&NiagaraSpriteRendererProperties, rendererIndex=3}
   {rendererClass=&NiagaraSpriteRendererProperties, rendererIndex=4}
  ]
 }
}
summary {
 systemName=NS_Emote_DairyString_RayLaser
 emitters [
  {emitterName=BlowingParticles, rendererClasses [&NiagaraRibbonRendererProperties]}
  {
   emitterName=GatheringFlare
   rendererClasses [
    &NiagaraSpriteRendererProperties
    &NiagaraLightRendererProperties
    &NiagaraMeshRendererProperties
   ]
  }
  {emitterName=ImpactFlare, rendererClasses [&NiagaraSpriteRendererProperties, &NiagaraLightRendererProperties]}
  {emitterName=BlowingParticles001, rendererClasses [&NiagaraSpriteRendererProperties]}
  {emitterName=Empty, rendererClasses []}
 ]
 userVariables [User.RayIndex {type=NiagaraFloat, defaultValue=1, description=""}]
}
